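# Physics asset for the ShirtTilapia gadget prop's skeletal mesh in BRCosmetics.
# Six ball joint bones each get an identical small sphere body, and the root bone gets one large sphere.
# Each ball constrains directly to root with identical swing and twist limits, forming a star layout rather than a chain.
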PhysicsAsset /BRCosmetics/Animation/Props/Gadgets_ShirtTilapia/Meshes/PA_ShirtTilapia_Gadget
bodies {
 JNT_Ball_01 [{shapeType=Sphere, radius=3.2385, center=<-0.0019 0 -0.1322>}]
 JNT_Ball_02 [{shapeType=Sphere, radius=3.2385, center=<-0.0019 0 -0.1322>}]
 JNT_Ball_03 [{shapeType=Sphere, radius=3.2385, center=<-0.0019 0 -0.1322>}]
 JNT_Ball_04 [{shapeType=Sphere, radius=3.2385, center=<-0.0019 0 -0.1322>}]
 JNT_Ball_05 [{shapeType=Sphere, radius=3.2385, center=<-0.0019 0 -0.1322>}]
 JNT_Ball_06 [{shapeType=Sphere, radius=3.2385, center=<-0.0019 0 -0.1322>}]
 root [{shapeType=Sphere, radius=103.455}]
}
constraints [
 {bone1Name=JNT_Ball_01, bone2Name=root, swing1LimitDegrees=45, swing2LimitDegrees=45, twistLimitDegrees=45}
 {bone1Name=JNT_Ball_02, bone2Name=root, swing1LimitDegrees=45, swing2LimitDegrees=45, twistLimitDegrees=45}
 {bone1Name=JNT_Ball_03, bone2Name=root, swing1LimitDegrees=45, swing2LimitDegrees=45, twistLimitDegrees=45}
 {bone1Name=JNT_Ball_04, bone2Name=root, swing1LimitDegrees=45, swing2LimitDegrees=45, twistLimitDegrees=45}
 {bone1Name=JNT_Ball_05, bone2Name=root, swing1LimitDegrees=45, swing2LimitDegrees=45, twistLimitDegrees=45}
 {bone1Name=JNT_Ball_06, bone2Name=root, swing1LimitDegrees=45, swing2LimitDegrees=45, twistLimitDegrees=45}
]
